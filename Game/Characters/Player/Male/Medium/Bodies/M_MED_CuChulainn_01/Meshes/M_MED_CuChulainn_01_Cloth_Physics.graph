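# Physics asset for the male medium CuChulainn_01 player body, named for cloth physics and apparently serving as cloth collision rather than a ragdoll.
# Single-capsule bodies cover pelvis, spine, neck, head, both clavicles and only the right upper arm, with spine_01 a tiny capsule.
# Constraints leave swing and twist free, and those for pelvis, spine_05, neck and head link each bone to itself, so no connected chain forms.
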PhysicsAsset /Game/Characters/Player/Male/Medium/Bodies/M_MED_CuChulainn_01/Meshes/M_MED_CuChulainn_01_Cloth_Physics
bodies {
 clavicle_l [{shapeType=Capsule, length=18.6013, radius=5.1879, center=<9.2638 -2.0344 -0.1476>, rotation=<rot 69.91 13.9949 -173.1531>}]
 clavicle_r [{shapeType=Capsule, length=18.6013, radius=5.1879, center=<-9.5637 1.8218 1.3119>, rotation=<rot 79.6883 -5.4523 -13.2192>}]
 head [{shapeType=Capsule, length=0, radius=12.75, center=<11.498 3.6669 0>, rotation=<rot 0 -90 0>}]
 neck_01 [{shapeType=Capsule, length=14.1925, radius=5.1253, center=<3.24 0 0>, rotation=<rot 90 0 0>}]
 pelvis [{shapeType=Capsule, length=6.2264, radius=12.1615}]
 spine_01 [{shapeType=Capsule, length=0.505, radius=0.505, rotation=<rot 0 -90 0>}]
 spine_02 [{shapeType=Capsule, length=9.8815, radius=11.3987}]
 spine_04 [{shapeType=Capsule, length=13.6638, radius=12.0972, center=<1.5459 -0.4822 0.0001>, rotation=<rot 0 -90 0>}]
 spine_05 [{shapeType=Capsule, length=13.1196, radius=11.6587, center=<-3.0023 -1.1128 -0.0001>}]
 upperarm_r [{shapeType=Capsule, length=8.4053, radius=6.3978, center=<-9.1531 -0.4746 -0.001>, rotation=<rot -88.2432 -86.823 89.7958>}]
}
constraints [
 {
  bone1Name=pelvis
  bone2Name=pelvis
  swing1LimitDegrees=45
  swing1Motion=Free
  swing2LimitDegrees=45
  swing2Motion=Free
  twistLimitDegrees=45
  twistMotion=Free
 }
 {
  bone1Name=spine_01
  bone2Name=pelvis
  swing1LimitDegrees=45
  swing1Motion=Free
  swing2LimitDegrees=45
  swing2Motion=Free
  twistLimitDegrees=45
  twistMotion=Free
 }
 {
  bone1Name=spine_02
  bone2Name=spine_01
  swing1LimitDegrees=45
  swing1Motion=Free
  swing2LimitDegrees=45
  swing2Motion=Free
  twistLimitDegrees=45
  twistMotion=Free
 }
 {
  bone1Name=spine_04
  bone2Name=spine_02
  swing1LimitDegrees=45
  swing1Motion=Free
  swing2LimitDegrees=45
  swing2Motion=Free
  twistLimitDegrees=45
  twistMotion=Free
 }
 {
  bone1Name=spine_05
  bone2Name=spine_05
  swing1LimitDegrees=45
  swing1Motion=Free
  swing2LimitDegrees=45
  swing2Motion=Free
  twistLimitDegrees=45
  twistMotion=Free
 }
 {
  bone1Name=neck_01
  bone2Name=neck_01
  swing1LimitDegrees=45
  swing1Motion=Free
  swing2LimitDegrees=45
  swing2Motion=Free
  twistLimitDegrees=45
  twistMotion=Free
 }
 {
  bone1Name=head
  bone2Name=head
  swing1LimitDegrees=45
  swing1Motion=Free
  swing2LimitDegrees=45
  swing2Motion=Free
  twistLimitDegrees=45
  twistMotion=Free
 }
 {
  bone1Name=clavicle_r
  bone2Name=spine_05
  swing1LimitDegrees=45
  swing1Motion=Free
  swing2LimitDegrees=45
  swing2Motion=Free
  twistLimitDegrees=45
  twistMotion=Free
 }
 {
  bone1Name=upperarm_r
  bone2Name=clavicle_r
  swing1LimitDegrees=45
  swing1Motion=Free
  swing2LimitDegrees=45
  swing2Motion=Free
  twistLimitDegrees=45
  twistMotion=Free
 }
 {
  bone1Name=clavicle_l
  bone2Name=spine_05
  swing1LimitDegrees=45
  swing1Motion=Free
  swing2LimitDegrees=45
  swing2Motion=Free
  twistLimitDegrees=45
  twistMotion=Free
 }
]
